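# Masked, Nanite-enabled material for the Jonesy tank prop in the LabRat chamber.
# Diffuse, normal and SRM maps blend against flat fallback values, with a masked tint and depth-based brightness.
# Eye-adaptation-inverse keeps base color and sine-pulsed emissive steady under auto exposure, with an emissive boost on low/mobile.
# Panning swirl noise, faded along local height, builds wind-style offset and normal perturbation; dithered masking and platform color switches are also present.
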Material /Hera/Props/LabRat/LabRat_Chamber_A/Jon/Materials/M_Jonesy_Tank
graph {
 properties {
  bAutomaticallySetUsageInEditor=false
  bUsedWithNanite=true
  blendMode=BLEND_Masked
  previewMesh=/Game/Effects/Fort_Effects/Effects/Environment/Hera/Shards/Glow/SM_Env_TeaCake_Shard_Glows
  usedShadingModels=Default Lit
 }
 nodes [
  n1 = TextureSampleParameter2D() {parameterName=Diffuse, texture=/Hera/Props/LabRat/LabRat_Chamber_A/Jon/Textures/T_M_MED_JonesyTank_Body_D}
  n2 = LinearInterpolate(A=VectorParameter() {parameterName=Base Color notxt, defaultValue=<rgba 1 1 1 0>}.RGB, B=@n1.RGB, Alpha=NamedRerouteUsage())
  n3 = Multiply(A=@n2, B=ScalarParameter() {defaultValue=1.5, parameterName=Multiply Mask})
  n4 = Saturate(None=Multiply(A=ScalarParameter() {parameterName=UseMask?}, B=Step(X=@n1.B) {constY=0.3}))
  n5 = VectorParameter() {parameterName=Vtx color, defaultValue=<rgba 0.5156 0.2534 0.1504 1>}
  n6 = NamedRerouteUsage()
  n7 = ChannelMaskParameter(None=@n6) {maskChannel=Alpha, parameterName=VTXMask, defaultValue=<rgba 0 0 0 1>}
  n8 = NamedRerouteUsage()
  n9 = Divide(
   A=Add(A=PixelDepth().R, B=ScalarParameter() {defaultValue=-550, parameterName=Param_2})
   B=ScalarParameter() {defaultValue=400, parameterName=Param_1}
  )
  n10 = Multiply(
   A=LinearInterpolate(A=LinearInterpolate(A=@n2, B=@n3, Alpha=@n4), B=@n5.RGB, Alpha=LinearInterpolate(B=@n7.R, Alpha=@n8))
   B=LinearInterpolate(B=ScalarParameter() {defaultValue=1.5, parameterName=Param_3}, Alpha=Saturate(None=@n9)) {constA=1}
  )
  n11 = VectorParameter() {parameterName=Emissive Multiply, defaultValue=<rgba 0 0 0 1>}
  n12 = Multiply(
   A=TextureSample() {texture=/Hera/Props/LabRat/LabRat_Chamber_A/Jon/Textures/T_M_MED_JonesyTank_Body_E}.G
   B=MaterialFunctionCall!(Default=@n11.RGB, Low/Mobile=Multiply(A=@n11.RGB) {constB=4}) {materialFunction=/Game/Effects/Fort_Effects/Materials/_Functions/MF_LowMobile_PathingSwitch}.Result
  )
  n13 = MaterialFunctionCall!(Sine Phase=Multiply(A=Time()) {constB=0.5}, Value 1=Constant() {r=1.1}, Value 2=Constant() {r=0.7}) {materialFunction=/Engine/Functions/Engine_MaterialFunctions02/Utility/Sine_Remapped}
  Add(
   A=EyeAdaptationInverse(LightValueInput=Multiply(A=@n10) {constB=0.15}, AlphaInput=Constant() {r=1}).EyeAdaptationInverse
   B=EyeAdaptationInverse(LightValueInput=Multiply(A=@n12, B=@n13.Remapped Sine), AlphaInput=Constant() {r=0.3}).EyeAdaptationInverse
  )
  ComponentMask(None=@n6) {r=true}
  Constant3Vector() {constant=<rgba 0 0 1 0>}
  n14 = WorldPosition() {worldPositionShaderOffset=WPT_ExcludeAllShaderOffsets}
  n15 = Subtract(A=ActorPositionWS(), B=@n14.XYZ)
  n16 = Divide(
   A=Add(
    A=Transform(None=@n15) {transformSourceType=TRANSFORMSOURCE_World, transformType=TRANSFORM_Local}
    B=ScalarParameter() {defaultValue=350, group=WindNew, parameterName=WPO_WindNew_Starting Point, sortPriority=5}
   )
   B=ScalarParameter() {defaultValue=155, group=WindNew, parameterName=WPO_WindNew_Fade, sortPriority=6}
  )
  n17 = ComponentMask(None=Max(A=Saturate(None=@n16), B=ScalarParameter() {group=WindNew, parameterName=WPO_Wind_Min})) {b=true}
  LinearInterpolate(A=Constant3Vector() {constant=<rgba 1 0 0 0>}, B=Constant3Vector() {constant=<rgba 0 1 0 0>}, Alpha=@n17)
  n18 = Constant() {r=1}
  n19 = Transform(None=@n14.XYZ) {transformSourceType=TRANSFORMSOURCE_World, transformType=TRANSFORM_Local}
  n20 = Constant3Vector() {constant=<rgba 256 256 0 0>}
  n21 = Abs(None=Constant3Vector() {constant=<rgba 512 512 384 0>})
  n22 = ComponentMask(None=Divide(A=Add(A=@n19, B=@n20) {constB=256}, B=Multiply(A=@n21) {constB=-1})) {b=true, r=true}
  n23 = ScalarParameter() {defaultValue=1.2, group=WindNew, parameterName=WPO_WindNew_Scale, sortPriority=1}
  n24 = ScalarParameter() {defaultValue=0.05, group=WindNew, parameterName=WPO_WindNew_Speed_X, sortPriority=2}
  n25 = ScalarParameter() {defaultValue=-0.1, group=WindNew, parameterName=WPO_WindNew_Speed_Y, sortPriority=2}
  n26 = TextureSampleParameter2D(UVs=Panner(Coordinate=Multiply(A=@n22, B=@n23), Speed=AppendVector(A=@n24, B=@n25))) {
   group=WindNew
   parameterName=WPO_WindNew_Texture
   samplerSource=SSM_Wrap_WorldGroupSettings
   samplerType=SAMPLERTYPE_Normal
   texture=/Game/Global/Textures/Default/Noise/T_SwirlNoise_N
  }
  n27 = AppendVector(
   A=AppendVector(
    A=ScalarParameter() {defaultValue=1, group=WindNew, parameterName=WPO_WindNew__X Val, sortPriority=3}
    B=ScalarParameter() {defaultValue=1, group=WindNew, parameterName=WPO_WindNew_Y Val, sortPriority=3}
   )
   B=Constant()
  )
  Multiply(A=@n17, B=Multiply(A=@n18, B=Multiply(A=@n26.RGB, B=@n27)))
  NamedRerouteDeclaration(None=ScalarParameter() {defaultValue=1, parameterName=Use Tex?}) {name=UseTEx?, nodeColor=<rgba 1 0.1875 0 1>}
  NamedRerouteDeclaration(None=MaterialFunctionCall(Alpha Threshold=NamedRerouteUsage()) {materialFunction=&DitherTemporalAA}.Result) {name=Masked, nodeColor=<rgba 0 0.25 1 1>}
  n28 = Constant3Vector() {constant=<rgba 0 0 1 0>}
  n29 = TextureSampleParameter2D() {parameterName=Normal, samplerType=SAMPLERTYPE_Normal, texture=/Hera/Props/LabRat/LabRat_Chamber_A/Jon/Textures/T_M_MED_JonesyTank_Body_N}
  n30 = LinearInterpolate(
   A=LinearInterpolate(A=@n28, B=@n29.RGB, Alpha=NamedRerouteUsage())
   B=@n28
   Alpha=ChannelMaskParameter(None=NamedRerouteUsage()) {maskChannel=Alpha, parameterName=VTXMask, defaultValue=<rgba 0 0 0 1>}.R
  )
  n31 = Multiply(
   A=Multiply(A=@n26.RGB, B=Constant3Vector() {constant=<rgba 1 1 0 0>})
   B=ScalarParameter() {defaultValue=0.1, group=WindNew, parameterName=WPO_WindNew_NormalsValue, sortPriority=4}
  )
  NamedRerouteDeclaration(None=LinearInterpolate(A=@n30, B=Add(A=@n30, B=@n31), Alpha=Multiply(A=@n17, B=@n18))) {name=Normal, nodeColor=<rgba 0.9219 1 0 1>}
  n32 = Add(A=LocalPosition().Z, B=ScalarParameter() {group=Pos GRad, parameterName=Pos Gradient Offset})
  n33 = ScalarParameter() {defaultValue=1, group=Pos GRad, parameterName=Pos Gradient Size}
  NamedRerouteDeclaration(None=Saturate(None=Divide(A=@n32, B=@n33))) {name=Pos Grad, nodeColor=<rgba 0 0.2266 1 1>}
  n34 = TextureSample() {samplerType=SAMPLERTYPE_Masks, texture=/Hera/Props/LabRat/LabRat_Chamber_A/Jon/Textures/T_Jonesy_VtxMask_RG}
  n35 = Constant() {r=1.5}
  NamedRerouteDeclaration(None=Saturate(None=Multiply(A=@n34.RGB, B=@n35))) {name=BakedVTX, nodeColor=<rgba 0 0.25 1 1>}
  NamedRerouteUsage()
  NamedRerouteUsage()
  Power()
  n36 = Multiply(A=@n5.RGB, B=ScalarParameter() {defaultValue=2.2705, parameterName=Param_4})
  QualitySwitch(
   Default=MaterialFunctionCall!("Xb1,PS4"=@n36, "PC-High&&Epic,XSX,PS5"=@n5.RGB, "PC-Perf,Android"=@n5.RGB, Switch=@n36, "PC-Low(NotPerf)&&Medium"=@n36) {materialFunction=/Game/Effects/Niagara/MasterMaterials/Functions/Common/PlatformSwitch_3D}.Result
   Low=@n36
   High=@n36
   Medium=@n36
  )
  RayTracingQualitySwitch()
  n37 = TextureSampleParameter2D() {parameterName=SRM, texture=/Hera/Props/LabRat/LabRat_Chamber_A/Jon/Textures/T_M_MED_JonesyTank_Body_SRM}
  n38 = NamedRerouteUsage()
  Reroute(None=LinearInterpolate(A=ScalarParameter() {parameterName=Spec NoTxt}, B=@n37.G, Alpha=@n38))
  n39 = ScalarParameter() {defaultValue=1.3, parameterName=Rioughness Multiply}
  Reroute(None=LinearInterpolate(A=ScalarParameter() {parameterName=Roughness NoTxt}, B=Multiply(A=@n37.G, B=@n39) {constB=0.6}, Alpha=@n38))
  ScalarParameter() {defaultValue=1}
  n40 = VertexColor()
  StaticComponentMaskParameter(None=AppendVector(A=@n40, B=@n40)) {defaultA=true, group=WindNew, parameterName=WPO_WindNew_VertexChannelMask, sortPriority=7}
  VertexColor()
  VertexColor()
 ]
}
